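; Parent Blueprint for an owner-only jewel in the Agent Jones quest puzzle, where jewels must be activated in order.
; Interaction plays a spin timeline that toggles the glass material and blocks interaction on all linked jewels via collision responses.
; The server then compares the jewel's index to the owner's GameplayEffect stack count, which tracks progress.
; Correct picks add progress or send the quest event on the last; wrong picks clear it and reverse activated jewels.
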(event EventBeginPlay
  ;[1 1 1] BeginPlay - Initialize various variables
  (|CallParentFunction)
  (Networking|SwitchHasAuthority
    (:Remote
      (CallFunction|SetIsOwnerLocallyControlled self)
      (Variables|BPVars|Data|SetOriginalMaterial (Utilities|Array|Get(acopy) (Rendering|Material|GetMaterials (Variables|Debug|GetStaticMeshComponent))))
      (Variables|BPVars|Data|SetIsGlass true)
      (Variables|BPVars|Data|SetZ%20%28Yaw%29 (.yaw (Transformation|GetActorRotation self)))
      (Collision|SetCollisionResponsetoChannel (Variables|Debug|GetStaticMeshComponent) "ECC_GameTraceChannel1" "ECR_Overlap"))))

(event Custom|ActivateJewel
  ;[1 1 1] Activate or inactivate the jewel along a timeline
  (Variables|BPVars|Data|SetIsInactive false)
  (CallFunction|PauseInteraction)
  (if (Variables|BPVars|Data|GetIsOwnerLocallyControlled)
    (Audio|SpawnSoundAttached (Variables|BPVars|Art|GetSpinSound) (Variables|Debug|GetStaticMeshComponent))
    (else
      _))
  ;;[1 1 1] Cosmetically spin the jewel
  (bind (_direction _rotation) (Timeline|SpinJewel PlayFromStart)
    (:Update
      (Networking|SwitchHasAuthority
        (:Remote
          (Transformation|SetActorRotation self (Math|Rotator|MakeRotator 0.0 0.0 (+ _rotation (Variables|BPVars|Data|GetZ%20%28Yaw%29))) false))))
    (:Finished
      (CallFunction|ResetInteraction)
      (if (Variables|BPVars|Data|GetIsInactive)
        ;;[1 1 1] Enable interact so the cycle can restart
        (Variables|BPVars|Data|SetIsGlass true)
        (Collision|SetCollisionResponsetoChannel (Variables|Debug|GetStaticMeshComponent) "ECC_GameTraceChannel1" "ECR_Overlap")
        (else
          ;;[1 1 1] Check if the jewel has been activated in the correct order
          (Variables|BPVars|Data|SetIsGlass false)
          (Networking|SwitchHasAuthority
            (:Authority
              (CallFunction|CheckStackCount))))))
    (:ToggleMaterial
      ;;[1 1 1] Swap the jewel between the default or a glassy material
      (if (Variables|BPVars|Data|GetIsOwnerLocallyControlled)
        (Variables|BPVars|Data|SetIsGlass (not (Variables|BPVars|Data|GetIsGlass)))))))

(event Custom|DeactivateJewel
  ;[1 1 1] Activate or inactivate the jewel along a timeline
  (Variables|BPVars|Data|SetIsInactive true)
  (CallFunction|PauseInteraction)
  (if (Variables|BPVars|Data|GetIsOwnerLocallyControlled)
    (Audio|SpawnSoundAttached (Variables|BPVars|Art|GetSpinSound) (Variables|Debug|GetStaticMeshComponent))
    (else
      _))
  ;;[1 1 1] Cosmetically spin the jewel
  (Timeline|SpinJewel ReverseFromEnd))

(event Interaction|EventBlueprintOnInteract
  (params (InteractingPawn :type Fort%20Pawn%20Object%20Reference) (InteractionBeingAttempted :type EInteractionBeingAttempted%20Enum :default "FirstInteraction"))
  ;[1 1 1] Activate or inactivate the jewel along a timeline
  ;;[1 1 1] On interact, we activate the jewel
  (CallFunction|ActivateJewel))

(event Custom|CheckStackCount
  ;[1 1 1] Check if the jewel was activated in the correct order compared to other jewels. We increment the count or give quest progress on success or reset all jewels on failure.
  (bind _asfort_player_pawn (Utilities|Casting|CastToFortPlayerPawn (Actor|GetOwner)))
  (if (== (Variables|BPVars|Data|GetGEStackCount) (GameplayEffects|GetGameplayEffectCount (Ability|GetFortAbilitySystemComponent _asfort_player_pawn) (Variables|BPVars|Data|GetJewelActivationGE) 0 true))
    (CallFunction|ProgressCosmetics)
    (Environment|ApplyGlobalEnvironmentGameplayEffecttoActor (Actor|GetOwner) (select (== (Variables|BPVars|Data|GetGEStackCount) 3) (Variables|BPVars|Data|GetQuestGE) (Variables|BPVars|Data|GetJewelActivationGE)))
    (if (== (Variables|BPVars|Data|GetGEStackCount) 3)
      (Quest|SendComplexCustomStatEvent (Quests|GetQuestManager (Fortnite|GetFortPlayerControllerfromActor (Actor|GetOwner)) "Invalid") self "(GameplayTags=)" (GameplayTags|GetOwnedGameplayTags self :!class /Script/GameplayTags.GameplayTagAssetInterface) 1))
    (else
      (GameplayEffects|RemoveActiveEffectswithGrantedTags (Ability|GetFortAbilitySystemComponent _asfort_player_pawn) "(GameplayTags=((TagName=\"Granted.Quest.AgentJones.JewelPuzzle.JewelCounter\")))")
      (CallFunction|DeactivateJewel)
      (for _array_element (Variables|BPVars|Data|GetOtherJewels)
        (if (not (Variables|BPVars|Data|GetIsInactive _array_element))
          (CallFunction|DeactivateJewel _array_element))))))

(event Custom|ProgressCosmetics
  ;[0.03368 1 0] Progress made cosmetics!
  (bind _returnvalue (Niagara|SpawnSystemAttached "/Game/Environments/Sets/ArtDeco/Props/FX/NS_Gems_Spawn.NS_Gems_Spawn" (Variables|Debug|GetStaticMeshComponent) "None" "0,0.500000,2.500000"))
  (Rendering|SetOnlyOwnerSee _returnvalue true :!class /Script/Engine.PrimitiveComponent)
  (if (Variables|BPVars|Data|GetIsOwnerLocallyControlled)
    (Audio|SpawnSoundAttached (Variables|BPVars|Art|GetProgressSound) (Variables|Debug|GetStaticMeshComponent))))

(fn BlueprintCanInteract
  (params (InteractingPawn :type Fort%20Pawn%20Object%20Reference) (InteractionBeingAttempted :type EInteractionBeingAttempted%20Enum :default "FirstInteraction") (InteractionType :type TInteractionType%20Enum :default "IT_NoInteraction"))
  (return (and (Utilities|Equal(Object) InteractingPawn (Actor|GetOwner)) (Variables|BPVars|Data|GetIsInactive))))

(fn BlueprintGetInteractionString
  (params (InteractingPawn :type Fort%20Pawn%20Object%20Reference) (InteractionBeingAttempted :type EInteractionBeingAttempted%20Enum :default "FirstInteraction"))
  (return (Variables|BPVars|Setup|GetInteractionString)))

(fn BlueprintOnLocalInteract
  (params (InteractingPawn :type Fort%20Player%20Pawn%20Object%20Reference) (InteractionBeingAttempted :type EInteractionBeingAttempted%20Enum :default "FirstInteraction"))
  (return false))

(fn OnRep_IsGlass
  (Rendering|Material|SetMaterial (Variables|Debug|GetStaticMeshComponent) 0 (select (Variables|BPVars|Data|GetIsGlass) (Variables|BPVars|Data|GetGlassMaterial) (Variables|BPVars|Data|GetOriginalMaterial))))

(fn PauseInteraction
  (Collision|SetCollisionResponsetoChannel (Variables|Debug|GetStaticMeshComponent) "ECC_GameTraceChannel1" "ECR_Ignore")
  (for _array_element (Variables|BPVars|Data|GetOtherJewels)
    (Collision|SetCollisionResponsetoChannel (Variables|/Script/FortniteGame.BuildingSMActor|GetStaticMeshComponent _array_element) "ECC_GameTraceChannel1" "ECR_Ignore"))
  (return))

(fn ResetInteraction
  (if (Variables|BPVars|Data|GetIsInactive)
    (Collision|SetCollisionResponsetoChannel (Variables|Debug|GetStaticMeshComponent) "ECC_GameTraceChannel1" "ECR_Overlap")
    (else
      _))
  (for _array_element (Variables|BPVars|Data|GetOtherJewels)
    (if (Variables|BPVars|Data|GetIsInactive _array_element)
      (Collision|SetCollisionResponsetoChannel (Variables|/Script/FortniteGame.BuildingSMActor|GetStaticMeshComponent _array_element) "ECC_GameTraceChannel1" "ECR_Overlap")))
  (return))

(fn SetIsOwnerLocallyControlled
  (params (ActorToCheck :type Actor%20Object%20Reference))
  (bind _array_element (Utilities|Array|ForEachLoopwithBreak (Fortnite|GetLocalFortPlayerControllers true))
    (:LoopBody
      (if (Utilities|Equal(Object) _array_element (Fortnite|GetFortPlayerControllerfromActor (Actor|GetOwner ActorToCheck)))
        (Variables|BPVars|Data|SetIsOwnerLocallyControlled true)))
    (:Completed
      _))
  (return))

(fn ConstructionScript
  (|CallParentFunction)
  (Rendering|SetOnlyOwnerSee (Variables|Debug|GetStaticMeshComponent) true :!class /Script/Engine.PrimitiveComponent))
