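; Quest spawner for the Agent Jones jewel puzzle that spawns jewel actors for qualifying players, and destroys itself if the creative-mode check fails.
; On server-side overlap, each unrecorded player pawn with the quest in progress, or the Test flag set, is stored and triggers SpawnJewels.
; Jewels spawn from class and transform arrays, offset from the spawner's location, with that pawn as instigator.
; Each jewel then receives references to its siblings.
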(event EventBeginPlay
  (bind _cancompletequest (Class|AthenaQuestFunctionLibrary|AthenaCheckCreativeMode (Game|GetGameState)))
  (if (not _cancompletequest)
    (Actor|DestroyActor)))

(event OnComponentBeginOverlap(SpawnCollision)
  (params (OverlappedComponent :type Primitive%20Component%20Object%20Reference) (OtherActor :type Actor%20Object%20Reference) (OtherComp :type Primitive%20Component%20Object%20Reference) (OtherBodyIndex :type Integer) (bFromSweep :type Boolean) (SweepResult :type Hit%20Result%20Structure%20%28by%20ref%29))
  (Networking|SwitchHasAuthority
    (:Authority
      (if (== (Utilities|Array|FindItem (Variables|Default|GetStoredPawns) OtherActor) -1)
        (bind _asfort_player_pawn (Utilities|Casting|CastToFortPlayerPawn OtherActor))
        (Utilities|IsValid (Fortnite|GetFortPlayerControllerfromActor _asfort_player_pawn)
          (:Is%20Valid
            (Utilities|IsValid (Quests|GetQuestManager (Fortnite|GetFortPlayerControllerfromActor _asfort_player_pawn) "Invalid")
              (:Is%20Valid
                (bind _cancompletequest (Class|AthenaQuestFunctionLibrary|AthenaCheckCreativeMode (Game|GetGameState)))
                (if _cancompletequest
                  (if (or (Quest|IsQuestinProgress (Quests|GetQuestManager (Fortnite|GetFortPlayerControllerfromActor _asfort_player_pawn) "Invalid") (Variables|Default|GetQuestID)) (Variables|Default|GetTest))
                    (Utilities|Array|AddUnique (Variables|Default|GetStoredPawns) _asfort_player_pawn)
                    (CallFunction|SpawnJewels _asfort_player_pawn)))))))))))

(fn GetLocationOffset
  (params (Index :type Integer))
  (bind _returnvalue_location (Math|Transform|BreakTransform (Transformation|GetActorTransform self)))
  (return (Math|Transform|MakeTransform (+ (Utilities|Array|Get(acopy) (Variables|Default|GetJewelLcationOffsets) Index) _returnvalue_location) (Utilities|Array|Get(acopy) (Variables|Default|GetJewelRotationOffsets) Index) (Utilities|Array|Get(acopy) (Variables|Default|GetJewelScales) Index))))

(fn SpawnJewels
  (params (Instigator :type Pawn%20Object%20Reference))
  (bind _jewels (LocalVariables|GetJewels))
  (for _index (range 4)
    (bind _returnvalue (Game|SpawnActorfromClass (Utilities|Array|Get(acopy) (Variables|Default|GetSpawnedActors) _index) (CallFunction|GetLocationOffset _index) "Undefined" "SelectDefaultAtRuntime" Instigator Instigator))
    (bind _asbp_quest_jewel_puzzle_jewel_parent (Utilities|Casting|CastToBP_Quest_JewelPuzzle_Jewel_Parent _returnvalue))
    (Utilities|Array|Add (LocalVariables|GetJewels) _asbp_quest_jewel_puzzle_jewel_parent))
  (if (Utilities|IsArrayNonEmpty _jewels)
    (for _array_element _jewels
      (Variables|/BattlepassS16/QuestActors/AgentJones/JewelPuzzle/BP_Quest_JewelPuzzle_Jewel_Parent.BP_Quest_JewelPuzzle_Jewel_Parent_C|SetOtherJewels (LocalVariables|GetJewels) _array_element)
      (Utilities|Array|RemoveItem (Variables|/BattlepassS16/QuestActors/AgentJones/JewelPuzzle/BP_Quest_JewelPuzzle_Jewel_Parent.BP_Quest_JewelPuzzle_Jewel_Parent_C|GetOtherJewels _array_element) _array_element))))

(fn ConstructionScript)
